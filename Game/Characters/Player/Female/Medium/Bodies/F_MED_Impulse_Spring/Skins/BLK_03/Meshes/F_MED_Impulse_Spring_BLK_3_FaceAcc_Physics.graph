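# Physics asset for the BLK_03 FaceAcc mesh of the F_MED_Impulse_Spring female character skin, covering its hair and ponytail.
# Five ponytail capsules form a chain with limited swing and tight twist, with the tip allowed a wider swing.
# The chain hangs from base and hair bodies linked to the head by free angular constraints.
# Head, neck, clavicle and upper-spine capsules approximate the body around the ponytail, also joined by free angular constraints.
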PhysicsAsset /Game/Characters/Player/Female/Medium/Bodies/F_MED_Impulse_Spring/Skins/BLK_03/Meshes/F_MED_Impulse_Spring_BLK_3_FaceAcc_Physics
bodies {
 blk3_hair [{shapeType=Capsule, length=1, radius=1}]
 blk3_ponytail_a [{shapeType=Capsule, length=0, radius=4.25, center=<2.3521 0 0.0001>, rotation=<rot 90 0 0>}]
 blk3_ponytail_a_base [{shapeType=Capsule, length=1, radius=1}]
 blk3_ponytail_b [{shapeType=Capsule, length=0, radius=4.25, center=<2.5372 0 0>, rotation=<rot 90 0 0>}]
 blk3_ponytail_c [{shapeType=Capsule, length=0, radius=4.25, center=<2.6589 0 0>, rotation=<rot 90 0 0>}]
 blk3_ponytail_d [{shapeType=Capsule, length=0.25, radius=3.75, center=<2.5175 0 0>, rotation=<rot 90 0 0>}]
 blk3_ponytail_e [{shapeType=Capsule, length=0.75, radius=3.75, center=<2.6436 0 0>, rotation=<rot 90 0 0>}]
 clavicle_l [{shapeType=Capsule, length=10.75, radius=5, center=<4.599 -0.9299 -0.1571>, rotation=<rot 82.8053 0 0>}]
 clavicle_r [{shapeType=Capsule, length=10.75, radius=5, center=<-4.599 0.9299 0.1571>, rotation=<rot 82.8053 0 -180>}]
 head [{shapeType=Capsule, length=2.75, radius=7.75, center=<5.2861 0.908 0>, rotation=<rot 90 0 0>}]
 neck_01 [{shapeType=Capsule, length=8.5, radius=7, center=<3.7431 -2.1937 0>, rotation=<rot 90 0 -15>}]
 spine_05 [
  {shapeType=Capsule, length=6.75, radius=8.5, center=<3.5203 0.8705 0>}
  {shapeType=Capsule, length=0, radius=17.25, shapeName=spine_05_capsule, center=<0.9186 1.8252 0>, rotation=<rot 90 180 155.0001>}
 ]
}
constraints [
 {
  bone1Name=blk3_ponytail_a_base
  bone2Name=blk3_hair
  swing1LimitDegrees=45
  swing1Motion=Free
  swing2LimitDegrees=45
  swing2Motion=Free
  twistLimitDegrees=45
  twistMotion=Free
 }
 {bone1Name=blk3_ponytail_a, bone2Name=blk3_ponytail_a_base, swing1LimitDegrees=30, swing2LimitDegrees=30, twistLimitDegrees=5}
 {bone1Name=blk3_ponytail_b, bone2Name=blk3_ponytail_a, swing1LimitDegrees=30, swing2LimitDegrees=30, twistLimitDegrees=5}
 {bone1Name=blk3_ponytail_c, bone2Name=blk3_ponytail_b, swing1LimitDegrees=30, swing2LimitDegrees=30, twistLimitDegrees=5}
 {bone1Name=blk3_ponytail_d, bone2Name=blk3_ponytail_c, swing1LimitDegrees=30, swing2LimitDegrees=30, twistLimitDegrees=5}
 {bone1Name=blk3_ponytail_e, bone2Name=blk3_ponytail_d, swing1LimitDegrees=45, swing2LimitDegrees=45, twistLimitDegrees=5}
 {
  bone1Name=blk3_hair
  bone2Name=head
  swing1LimitDegrees=45
  swing1Motion=Free
  swing2LimitDegrees=45
  swing2Motion=Free
  twistLimitDegrees=45
  twistMotion=Free
 }
 {
  bone1Name=clavicle_l
  bone2Name=spine_05
  swing1LimitDegrees=45
  swing1Motion=Free
  swing2LimitDegrees=45
  swing2Motion=Free
  twistLimitDegrees=45
  twistMotion=Free
 }
 {
  bone1Name=clavicle_r
  bone2Name=spine_05
  swing1LimitDegrees=45
  swing1Motion=Free
  swing2LimitDegrees=45
  swing2Motion=Free
  twistLimitDegrees=45
  twistMotion=Free
 }
 {
  bone1Name=neck_01
  bone2Name=spine_05
  swing1LimitDegrees=45
  swing1Motion=Free
  swing2LimitDegrees=45
  swing2Motion=Free
  twistLimitDegrees=45
  twistMotion=Free
 }
 {
  bone1Name=head
  bone2Name=neck_01
  swing1LimitDegrees=45
  swing1Motion=Free
  swing2LimitDegrees=45
  swing2Motion=Free
  twistLimitDegrees=45
  twistMotion=Free
 }
]
